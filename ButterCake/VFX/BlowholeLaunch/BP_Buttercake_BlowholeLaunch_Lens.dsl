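; Buttercake blowhole launch lens effect Blueprint, which spawns its VFX components at runtime.
; On BeginPlay it calls the parent, then adds a Niagara component and a particle system component, explicitly activating the latter without reset.
; Components are added in the graph rather than the component outliner because, per an in-graph note, adding them there crashes.
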(event EventBeginPlay
  (|CallParentFunction)
  ;[1 1 1] Adding a component in the outliner crashes
  (AddComponent|NODE_AddNiagaraComponent-0)
  (bind _returnvalue (AddComponent|NODE_AddParticleSystemComponent-1))
  (Components|Activation|Activate _returnvalue false))

(fn ConstructionScript)
