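# Translucent unlit Niagara sprite material for the mist of the blowhole launch geyser.
# Two panning smoke textures, warped by a distortion normal map, blend into a smoke pattern, with a particle-random time offset declared for variation.
# A panning trail mask, bowed vertically by a sphere mask, combines with particle alpha to shape opacity.
# Dynamic parameters drive bend, mask pan and smoke amount; particle color feeds an emissive adjustment function.
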Material /ButterCake/VFX/BlowholeLaunch/M_BlowholeGeyserMist
graph {
 properties {bUsedWithNiagaraSprites=true, blendMode=BLEND_Translucent, pixelDepthOffsetMode=PDOM_Legacy, shadingModel=MSM_Unlit, usedShadingModels=Unlit}
 nodes [
  n1 = ParticleColor()
  MaterialFunctionCall(Emissive=@n1.RGB) {materialFunction=&MF_FX_EmissiveAdjustments}
  Multiply()
  n2 = SphereMask(A=ComponentMask(None=TextureCoordinate()) {r=true}, B=Constant() {r=0.5}) {attenuationRadius=0.5, hardnessPercent=0}
  n3 = DynamicParameter() {paramNames [ShiftUpDown, MaskPanLeftRightSpeed, SmokeAmount, Param4]}
  n4 = AppendVector(A=Constant(), B=@n3.ShiftUpDown)
  n5 = TextureCoordinate() {uTiling=0.75}
  n6 = AppendVector(A=NamedRerouteUsage(), B=Constant())
  n7 = Panner(Coordinate=Add(A=Multiply(A=@n2, B=@n4), B=Add(A=@n5, B=@n6)), Speed=AppendVector(A=@n3.MaskPanLeftRightSpeed, B=Constant())) {speedX=0.672}
  n8 = ComponentMask(None=@n7) {g=true}
  n9 = TextureSample(UVs=AppendVector(A=ComponentMask(None=@n7) {r=true}, B=Saturate(None=@n8))) {texture=/Game/Effects/Fort_Effects/Textures/Smoke/T_Lobber_Trail_Mask}
  n10 = NamedRerouteUsage()
  Multiply(A=@n1.A, B=Multiply(A=Add(A=@n9.G, B=@n10), B=NamedRerouteUsage()))
  n11 = Constant() {r=0.5}
  n12 = ScalarParameter() {parameterName=MaskDistort}
  n13 = Add(A=@n11, B=@n12)
  n14 = Subtract(A=@n11, B=@n12)
  n15 = Saturate(None=NamedRerouteUsage())
  NamedRerouteDeclaration(None=SphereMask(A=NamedRerouteUsage(), B=Constant() {r=0.5}, Radius=LinearInterpolate(A=@n13, B=@n14, Alpha=@n15)) {attenuationRadius=0.5, hardnessPercent=0}) {name=AlphaMask, nodeColor=<rgba 1 0.8438 0 1>}
  NamedRerouteDeclaration(None=TextureCoordinate()) {name=0-1 UV, nodeColor=<rgba 1 0.8438 0 1>}
  n16 = Multiply(A=TextureCoordinate(), B=ScalarParameter() {defaultValue=1, parameterName=TextureScale})
  n17 = VectorParameter() {parameterName=SmokeScaling, defaultValue=<rgba 1 0.268 0 -1.5>}
  n18 = VectorParameter() {group=WPO, parameterName="DistortionTexture_Tile&&Speed", defaultValue=<rgba 0.99 0.5 0 0>}
  n19 = ComponentMask(None=@n18.RGB) {g=true, r=true}
  n20 = TextureSampleParameter2D(UVs=Panner(Coordinate=Multiply(A=@n16, B=@n19), Speed=AppendVector(A=@n18.B, B=@n18.B)) {speedY=0.1}) {parameterName=WPO, samplerType=SAMPLERTYPE_Normal, texture=/Game/Effects/Fort_Effects/Textures/T_DustDistortion}
  n21 = Multiply(
   A=NamedRerouteDeclaration(None=ComponentMask(None=@n20.RGB) {g=true, r=true}) {name=NoiseDistort, nodeColor=<rgba 0.9428 0 1 1>}
   B=ScalarParameter() {parameterName=Nrm_Distortion}
  )
  n22 = Add(A=Multiply(A=@n16, B=ComponentMask(None=@n17.RGB) {g=true, r=true}), B=@n21)
  n23 = NamedRerouteUsage()
  n24 = AppendVector(A=@n17.B, B=@n17.B)
  n25 = VectorParameter() {parameterName=SmokeScaling_2nd, defaultValue=<rgba 1 0.268 0 -1.5>}
  n26 = Add(A=Multiply(A=ComponentMask(None=@n25.RGB) {g=true, r=true}, B=TextureCoordinate()), B=@n21)
  n27 = AppendVector(A=@n25.B, B=@n25.B)
  n28 = LinearInterpolate(
   A=TextureSample(UVs=Panner(Coordinate=@n22, Time=@n23, Speed=@n24) {speedX=0.1, speedY=0.7}) {texture=&T_Smooth_tile}.R
   B=TextureSample(UVs=Panner(Coordinate=@n26, Time=@n23, Speed=@n27) {speedX=-0.1, speedY=1}) {texture=&T_Inky_Smoke_Tile}.R
  )
  NamedRerouteDeclaration(None=Add(A=@n28, B=DynamicParameter() {paramNames [ShiftUpDown, MaskPanLeftRightSpeed, SmokeAmount, Param4]}.SmokeAmount)) {name=SmokeyPattern, nodeColor=<rgba 0.1719 1 0 1>}
  NamedRerouteDeclaration(None=Add(A=ParticleRandom(), B=Time())) {name=PerParticleTime, nodeColor=<rgba 1 0.8438 0 1>}
  NamedRerouteDeclaration(None=ParticleRandom()) {name=PerParticleRandVal, nodeColor=<rgba 1 0.8438 0 1>}
  NamedRerouteDeclaration(None=@n28) {name=RawSmoke, nodeColor=<rgba 0.1719 1 0 1>}
  Power(Base=@n2) {constExponent=0.5}
  ScalarParameter() {parameterName=Tex_BrightnessOffset}
  ScalarParameter() {parameterName=Smoke1_Scale}
  ScalarParameter() {parameterName=Smoke2_Scale}
  ScalarParameter() {parameterName=ShiftUpDown}
  TextureCoordinate() {vTiling=0.5}
 ]
}
