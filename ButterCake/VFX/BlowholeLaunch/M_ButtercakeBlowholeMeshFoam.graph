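# Translucent lit particle material for foam on Buttercake's blowhole launch VFX meshes.
# Panning radial-UV smoke, distorted by waterfall textures, combines with a mask thresholded by dynamic FoamPosition to blend particle color toward foam color; time-of-day values scale brightness and emissive.
# Radial noise, vertex red and particle alpha drive opacity, with depth fade skipped on mobile feature levels.
# Panning noise, vertex red and dynamic WPO_Max_Z squared form a Z-axis vertex offset.
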Material /ButterCake/VFX/BlowholeLaunch/M_ButtercakeBlowholeMeshFoam
graph {
 properties {
  bEnableResponsiveAA=true
  bUsedWithMeshParticles=true
  bUsedWithNiagaraMeshParticles=true
  bUsedWithParticleSprites=true
  blendMode=BLEND_Translucent
  pixelDepthOffsetMode=PDOM_Legacy
  translucencyLightingMode=TLM_VolumetricPerVertexDirectional
  usedShadingModels=Default Lit
 }
 nodes [
  n1 = ScalarParameter() {defaultValue=0.5, parameterName=Min Opacity}
  n2 = ScalarParameter() {defaultValue=1, parameterName=Max Opacity}
  n3 = ComponentMask(None=VectorParameter() {parameterName=NoisePanSpeed, defaultValue=<rgba 0 -1 0 0>}.RGB) {g=true, r=true}
  n4 = TextureSample(UVs=Panner(
    Coordinate=MaterialFunctionCall() {materialFunction=&VectorToRadialValue}.Radial Coordinates
    Speed=@n3
   ) {speedY=-1.6}) {texture=&T_NoisePack_01}
  n5 = Multiply(A=Multiply(A=VertexColor().R, B=LinearInterpolate(A=@n1, B=@n2, Alpha=@n4.R)), B=ParticleColor().A)
  n6 = DepthFade(Opacity=@n5, FadeDistance=ScalarParameter() {defaultValue=5, parameterName=DepthFade Distance})
  FeatureLevelSwitch(Default=@n6, ES2=@n5, ES3_1=@n5, SM4_REMOVED=@n6, SM5=@n6, SM6=@n6)
  n7 = MaterialFunctionCall(
   Value_Morning=ScalarParameter() {defaultValue=0.05, parameterName=Morning_Emiss}
   Value_Day=ScalarParameter() {defaultValue=0.2, parameterName=Day_Emiss}
   Value_Evening=ScalarParameter() {defaultValue=0.1, parameterName=Evening_Emiss}
   Value_Night=ScalarParameter() {defaultValue=0.075, parameterName=Night_Emiss}
  ) {materialFunction=&Scalar_Time_of_Day}
  n8 = ParticleColor()
  n9 = VectorParameter() {group=Foam, parameterName=FoamColor, defaultValue=<rgba 1 1 1 1>}
  n10 = DynamicParameter() {paramNames [WPO_Max_Z_Scale, WPO_Max_Z, FoamPosition, Param4]}
  n11 = ScalarParameter() {defaultValue=0.25, group=Foam, parameterName=FoamMaskWidth}
  n12 = TextureCoordinate()
  n13 = ScalarParameter() {defaultValue=1, group=Foam, parameterName=FoamMaskTex_Scale}
  n14 = TextureSampleParameter2D(UVs=VertexInterpolator(VS=MaterialFunctionCall(UVs=@n12, Texture Scale=@n13) {materialFunction=&ScaleUVsByCenter}.UVs).PS) {group=Foam, parameterName=T_FoamMask, samplerSource=SSM_Clamp_WorldGroupSettings, texture=&T_NoisePack_01}
  n15 = ScalarParameter() {defaultValue=1, parameterName=Waterfall_UScale}
  n16 = ScalarParameter() {defaultValue=1, parameterName=Waterfall_VScale}
  n17 = Constant() {r=0.5}
  n18 = Subtract(A=Multiply(A=TextureCoordinate(), B=AppendVector(A=@n15, B=@n16)), B=AppendVector(A=Multiply(A=@n15, B=@n17), B=Multiply(A=@n16, B=@n17)))
  n19 = VertexInterpolator(VS=NamedRerouteDeclaration(None=MaterialFunctionCall(Vector or UVs=@n18) {materialFunction=&VectorToRadialValue}.Radial Coordinates) {name=RadialUV, nodeColor=<rgba 0.1016 0 1 1>})
  n20 = Panner(Coordinate=@n19.PS) {speedY=0.125}
  n21 = Panner(Coordinate=@n19.PS) {speedY=-0.125}
  n22 = Add(
   A=Multiply(
    A=TextureSample(UVs=@n20) {texture=/Game/Packages/Fortnite_UniqueMaterials_1/Textures/T_WaterFall_Tile}.B
    B=ScalarParameter() {defaultValue=0.0163, group=WaterFall_UVDistort, parameterName=UVDistortAmountA}
   )
   B=Multiply(
    A=TextureSample(UVs=@n21) {texture=&T_WaterFall_Tile}.R
    B=ScalarParameter() {defaultValue=0.0163, group=WaterFall_UVDistort, parameterName=UVDistortAmountB}
   )
  )
  n23 = TextureSample(UVs=Panner(Coordinate=Add(A=@n19.PS, B=@n22), Speed=@n3) {speedY=-1}) {samplerType=SAMPLERTYPE_LinearColor, texture=/Game/Global/Textures/Default/Tileable_Patterns/T_Tile_Smoke_Small}
  n24 = Multiply(A=@n9.A, B=SmoothStep(Min=@n10.FoamPosition, Max=Add(A=@n10.FoamPosition, B=@n11), Value=Multiply(A=@n14.R, B=@n23.R)))
  n25 = MaterialFunctionCall(
   Value_Morning=ScalarParameter() {defaultValue=0.4, parameterName=Morning}
   Value_Day=ScalarParameter() {defaultValue=0.75, parameterName=Day}
   Value_Evening=ScalarParameter() {defaultValue=0.5, parameterName=Evening}
   Value_Night=ScalarParameter() {defaultValue=0.85, parameterName=Night}
  ) {materialFunction=&Scalar_Time_of_Day}
  n26 = Multiply(
   A=Multiply(A=LinearInterpolate(A=@n8.RGB, B=@n9.RGB, Alpha=@n24), B=VertexInterpolator(VS=@n25.Result).PS) {constB=0.75}
   B=Saturate(None=Add(A=@n23.RGB) {constB=0.25})
  )
  MaterialFunctionCall(Emissive=Multiply(A=@n26, B=VertexInterpolator(VS=@n7.Result).PS) {constB=0.2}) {materialFunction=&MF_FX_EmissiveAdjustments}
  n27 = DynamicParameter() {paramNames [WPO_Max_Z_Scale, WPO_Max_Z, FoamPosition, Param4]}
  n28 = Multiply(A=@n27.WPO_Max_Z, B=@n27.WPO_Max_Z)
  n29 = TextureSample(UVs=Panner(Coordinate=NamedRerouteUsage()) {speedY=-1}) {samplerType=SAMPLERTYPE_LinearColor, texture=&T_NoisePack_LinearColor}
  n30 = Constant3Vector() {desc="Buttercake: this should be world space - as when he sneezes the animations may rotate  the source ", constant=<rgba 0 0 1 0>}
  Multiply(A=Multiply(A=LinearInterpolate(B=@n28, Alpha=@n29.G), B=VertexColor().R), B=@n30)
  Reroute(None=@n26)
  VertexNormalWS()
 ]
}
