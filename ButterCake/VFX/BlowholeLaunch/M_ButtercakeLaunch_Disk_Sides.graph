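# Unlit, translucent, two-sided particle material for the sides of the smoke disk in the ButterCake blowhole launch effect.
# A horizontally tiled smoke disk texture is multiplied by a panning puff pattern, with dynamic parameters driving both tilings and the blend toward that eroded result.
# Particle color tints the color, which passes through day/night bloom control, and particle alpha scales opacity.
# Depth fade applies only at High and Epic quality.
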Material /ButterCake/VFX/BlowholeLaunch/M_ButtercakeLaunch_Disk_Sides
graph {
 properties {
  bUsedWithNiagaraMeshParticles=true
  bUsedWithNiagaraSprites=true
  blendMode=BLEND_Translucent
  pixelDepthOffsetMode=PDOM_Legacy
  shadingModel=MSM_Unlit
  twoSided=true
  usedShadingModels=Unlit
 }
 nodes [
  n1 = ParticleColor()
  n2 = DynamicParameter() {defaultValue=<rgba 0 0 0 1>, paramNames [Param1,Tiling_X,ErosionTile,Param4]}
  n3 = Constant() {r=1}
  n4 = TextureSample(UVs=Multiply(A=TextureCoordinate(), B=AppendVector(A=@n2.Tiling_X, B=@n3))) {texture=/Game/Effects/Fort_Effects/Textures/Smoke/T_Smoke_Bomb_Disk}
  n5 = Multiply(
   A=TextureCoordinate() {uTiling=10, vTiling=2}
   B=DynamicParameter() {defaultValue=<rgba 0 0 0 1>, paramNames [Param1,Tiling_X,ErosionTile,Param4]}.ErosionTile
  )
  n6 = Power(Base=TextureSample(UVs=Panner(Coordinate=@n5) {speedX=0.02, speedY=0.25}) {texture=/Game/Global/Textures/Default/Tileable_Patterns/T_Smoke_Puff_Tile_Up}.RGB)
  n7 = Multiply(A=Multiply(A=Clamp(None=@n6), B=@n4.RGB)) {constB=2}
  n8 = DynamicParameter() {defaultValue=<rgba 0 0 0 1>, paramNames [Param1,Tiling_X,ErosionTile,Param4]}
  MaterialFunctionCall!(DayBloom_Mult=Constant() {r=1}, In=Multiply(A=@n1.RGB, B=LinearInterpolate(A=@n4.RGB, B=@n7, Alpha=@n8.Param1))) {materialFunction=/Game/Packages/Fortnite_UniqueMaterials_1/MH_Materials/DayNight_Auto_Bloom_Control}
  n9 = Multiply(A=@n1.A, B=LinearInterpolate(A=@n4.A, B=@n7, Alpha=@n8.Param1))
  n10 = DepthFade(Opacity=@n9) {fadeDistanceDefault=30}
  QualitySwitch(Default=@n9, Low=@n9, High=@n10, Medium=@n9, Epic=@n10)
  TextureCoordinate()
 ]
}
