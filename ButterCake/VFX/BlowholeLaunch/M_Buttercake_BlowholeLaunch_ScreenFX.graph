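# Translucent unlit Niagara sprite material for a screen-space smoke overlay in the Buttercake blowhole launch effect.
# Two panning, rotated smoke textures on square screen UVs, distorted by a panning normal map, blend into the smoke pattern.
# Color combines particle color with time-of-day tints computed through a vertex interpolator.
# Opacity math combines particle alpha, a screen-space sphere mask and a vertical rolling mask driven by dynamic parameters.
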Material /ButterCake/VFX/BlowholeLaunch/M_Buttercake_BlowholeLaunch_ScreenFX
graph {
 properties {bUsedWithNiagaraSprites=true, blendMode=BLEND_Translucent, pixelDepthOffsetMode=PDOM_Legacy, shadingModel=MSM_Unlit, usedShadingModels=Unlit}
 nodes [
  MaterialFunctionCall(Emissive=Multiply(A=NamedRerouteUsage(), B=ParticleColor().RGB)) {materialFunction=&MF_FX_EmissiveAdjustments}
  n1 = Multiply(
   A=SphereMask(A=ScreenPosition().ViewportUV, B=Constant2Vector() {g=0.65, r=0.5}) {attenuationRadius=1, hardnessPercent=0.5}
   B=ParticleColor().A
  )
  n2 = NamedRerouteUsage()
  n3 = ComponentMask(None=ScreenPosition().ViewportUV) {g=true}
  n4 = DynamicParameter() {paramNames [MaxOpacity, RollingOpacity_Position, RollingOpacity_Width, Param4]}
  Multiply(
   A=Multiply(A=Multiply(A=@n1, B=@n2), B=DynamicParameter() {paramNames [MaxOpacity, RollingOpacity_Position, RollingOpacity_Width, Param4]}.MaxOpacity)
   B=Power(Base=SphereMask(A=@n3, B=@n4.RollingOpacity_Position, Radius=@n4.RollingOpacity_Position) {hardnessPercent=0}) {constExponent=4}
  )
  n5 = MaterialFunctionCall(
   Color_Morning=VectorParameter() {parameterName=MorningTint, defaultValue=<rgba 1.15 0.8884 0.805 0>}.RGB
   Color_Day=VectorParameter() {parameterName=DayTint, defaultValue=<rgba 0.2874 0.6512 1.2 0>}.RGB
   Color_Evening=VectorParameter() {parameterName=EveningTint, defaultValue=<rgba 1.1 0.7845 0.77 1>}.RGB
   Color_Night=VectorParameter() {parameterName=NightTint, defaultValue=<rgba 0.2656 0.85 0.85 0>}.RGB
  ) {materialFunction=&Color_Time_of_Day}
  NamedRerouteDeclaration(None=VertexInterpolator(VS=LinearInterpolate(B=@n5.Result) {constA=1}).PS) {name=FinalColor, nodeColor=<rgba 0 0.3906 1 1>}
  NamedRerouteDeclaration(None=MaterialFunctionCall() {materialFunction=/FortUILibrary/Foundation/Art_ToolKit/Tools/MF_UI_SquareToScreenSpace}.Y_Priority) {name=ScreenAlignedSqUV, nodeColor=<rgba 1 0 0.8672 1>}
  n6 = VectorParameter() {parameterName=SmokeScaling, defaultValue=<rgba 1 0.268 0 -1.5>}
  n7 = Multiply(A=NamedRerouteUsage(), B=ComponentMask(None=@n6.RGB) {g=true, r=true})
  n8 = VectorParameter() {group=WPO, parameterName="DistortionTexture_Tile&&Speed", defaultValue=<rgba 0.99 0.5 0 0>}
  n9 = Panner(
   Coordinate=Multiply(A=TextureCoordinate() {vTiling=0.5}, B=ComponentMask(None=@n8.RGB) {g=true, r=true})
   Speed=AppendVector(A=@n8.B, B=@n8.B)
  ) {speedY=0.1}
  n10 = Multiply(
   A=ComponentMask(None=TextureSampleParameter2D(UVs=@n9) {parameterName=WPO, samplerType=SAMPLERTYPE_Normal, texture=/Game/Effects/Fort_Effects/Textures/T_DustDistortion}.RGB) {g=true, r=true}
   B=ScalarParameter() {parameterName=Nrm_Distortion}
  )
  n11 = Add(A=@n7, B=@n10)
  n12 = AppendVector(A=@n6.B, B=@n6.B)
  n13 = TextureSample(UVs=MaterialFunctionCall(
    UVs=Panner(Coordinate=@n11, Speed=@n12) {speedX=0.1, speedY=0.7}
    "Rotation Angle (0-1)"=ScalarParameter() {parameterName="Rotation Angle (0-1) (S)"}
   ) {materialFunction=&CustomRotator}.Rotated Values) {texture=&T_Smooth_tile}
  n14 = VectorParameter() {parameterName=SmokeScaling_2nd, defaultValue=<rgba 1 0.268 0 -1.5>}
  n15 = Add(A=Multiply(A=ComponentMask(None=@n14.RGB) {g=true, r=true}, B=NamedRerouteUsage()), B=@n10)
  n16 = AppendVector(A=@n14.B, B=@n14.B)
  n17 = TextureSample(UVs=MaterialFunctionCall(
    UVs=Panner(Coordinate=@n15, Speed=@n16) {speedX=-0.1, speedY=1}
    "Rotation Angle (0-1)"=ScalarParameter() {defaultValue=0.5, parameterName="Rotation Angle (0-1) (S)"}
   ) {materialFunction=&CustomRotator}.Rotated Values) {texture=&T_Inky_Smoke_Tile}
  NamedRerouteDeclaration(None=Add(A=LinearInterpolate(A=@n13.R, B=@n17.R), B=ScalarParameter() {parameterName=Tex_BrightnessOffset})) {name=SmokeyPattern, nodeColor=<rgba 0.1719 1 0 1>}
  TextureCoordinate()
  TextureSample(UVs=Add(A=Panner(Coordinate=@n7, Speed=@n12), B=@n10)) {texture=&T_Smooth_tile}
 ]
}
