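# Translucent unlit Niagara sprite material for screen-space spikes, named for the ButterCake blowhole launch effect.
# A scalable smoke-disk texture supplies the spike pattern, with panned tiling smoke available as a dynamic-parameter blend.
# Opacity math combines screen-position sphere masks, particle alpha and dynamic MaxOpacity and rolling-position controls.
# Color logic includes particle color, time-of-day tints evaluated through a vertex interpolator, and day/night bloom control.
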Material /ButterCake/VFX/BlowholeLaunch/M_Buttercake_Launched_ScreenSpikes
graph {
 properties {bUsedWithNiagaraSprites=true, blendMode=BLEND_Translucent, pixelDepthOffsetMode=PDOM_Legacy, shadingModel=MSM_Unlit, usedShadingModels=Unlit}
 nodes [
  Constant3Vector()
  n1 = VectorParameter() {parameterName=SpikeScaling, defaultValue=<rgba 1 1 0 0>}
  n2 = ComponentMask(None=@n1.RGB) {g=true, r=true}
  n3 = TextureCoordinate() {vTiling=0.5}
  n4 = TextureSample(UVs=Add(A=Multiply(A=@n2, B=@n3), B=AppendVector(A=@n1.B, B=@n1.B))) {texture=/Game/Effects/Fort_Effects/Textures/Smoke/T_Smoke_Bomb_Disk}
  n5 = Panner(Coordinate=TextureCoordinate() {uTiling=10, vTiling=2}) {speedX=0.02, speedY=0.25}
  n6 = Clamp(None=Power(Base=TextureSample(UVs=@n5) {texture=/Game/Global/Textures/Default/Tileable_Patterns/T_Smoke_Puff_Tile_Up}.RGB))
  LinearInterpolate(A=@n4.RGB, B=Multiply(A=Multiply(A=@n6, B=@n4.RGB)) {constB=2}, Alpha=DynamicParameter() {defaultValue=<rgba 0 0 0 1>}.Param1)
  n7 = ParticleColor()
  MaterialFunctionCall(Emissive=@n7.RGB) {materialFunction=&MF_FX_EmissiveAdjustments}
  Multiply(A=NamedRerouteUsage(), B=@n7.RGB)
  Multiply(A=@n3)
  n8 = Multiply(
   A=SphereMask(A=ScreenPosition().ViewportUV, B=Constant2Vector() {g=0.65, r=0.5}) {attenuationRadius=1, hardnessPercent=0.5}
   B=ParticleColor().A
  )
  n9 = NamedRerouteUsage()
  n10 = ComponentMask(None=ScreenPosition().ViewportUV) {g=true}
  n11 = DynamicParameter() {paramNames [MaxOpacity, RollingOpacity_Position, RollingOpacity_Width, Param4]}
  Multiply(
   A=Multiply(A=Multiply(A=@n8, B=@n9), B=DynamicParameter() {paramNames [MaxOpacity, RollingOpacity_Position, RollingOpacity_Width, Param4]}.MaxOpacity)
   B=Power(Base=SphereMask(A=@n10, B=@n11.RollingOpacity_Position, Radius=@n11.RollingOpacity_Position) {hardnessPercent=0}) {constExponent=4}
  )
  n12 = MaterialFunctionCall(
   Color_Morning=VectorParameter() {parameterName=MorningTint, defaultValue=<rgba 1.15 0.8884 0.805 0>}.RGB
   Color_Day=VectorParameter() {parameterName=DayTint, defaultValue=<rgba 0.2874 0.6512 1.2 0>}.RGB
   Color_Evening=VectorParameter() {parameterName=EveningTint, defaultValue=<rgba 1.1 0.7845 0.77 1>}.RGB
   Color_Night=VectorParameter() {parameterName=NightTint, defaultValue=<rgba 0.2656 0.85 0.85 0>}.RGB
  ) {materialFunction=&Color_Time_of_Day}
  NamedRerouteDeclaration(None=VertexInterpolator(VS=LinearInterpolate(B=@n12.Result) {constA=1}).PS) {name=FinalColor, nodeColor=<rgba 0 0.3906 1 1>}
  NamedRerouteDeclaration(None=MaterialFunctionCall() {materialFunction=/FortUILibrary/Foundation/Art_ToolKit/Tools/MF_UI_SquareToScreenSpace}.Y_Priority) {name=ScreenAlignedSqUV, nodeColor=<rgba 1 0 0.8672 1>}
  NamedRerouteDeclaration(None=@n4.R) {name=SpikePattern, nodeColor=<rgba 0.1719 1 0 1>}
  n13 = ParticleColor()
  NamedRerouteDeclaration(None=MaterialFunctionCall!(DayBloom_Mult=Constant() {r=1}, In=Multiply(A=@n13.RGB, B=@n4.R)) {materialFunction=/Game/Packages/Fortnite_UniqueMaterials_1/MH_Materials/DayNight_Auto_Bloom_Control}.Result) {name=FinalColor_1, nodeColor=<rgba 0 0.3906 1 1>}
  TextureCoordinate()
 ]
}
