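# Translucent unlit splash material for Niagara and particle sprites, filed with the blowhole launch VFX.
# A tiled, panning foam texture, offset per particle, is multiplied by particle color and time-of-day color, then passed through emissive adjustments.
# Foam alpha is shaped by a sphere mask with noise-varied radius, eroded by a dynamic parameter and scaled by particle alpha.
# The ES2 path samples the foam without panning or random offset.
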Material /ButterCake/VFX/BlowholeLaunch/M_Splash
graph {
 properties {
  bEnableResponsiveAA=true
  bUsedWithNiagaraSprites=true
  bUsedWithParticleSprites=true
  blendMode=BLEND_Translucent
  pixelDepthOffsetMode=PDOM_Legacy
  shadingModel=MSM_Unlit
  usedShadingModels=Unlit
 }
 nodes [
  DynamicParameter() {defaultValue=<rgba 0 1 1 1>, paramNames [Erosion,Random,Param3,Param4]}
  n1 = Multiply(A=TextureCoordinate(), B=ComponentMask(None=VectorParameter() {parameterName=FoamTile, defaultValue=<rgba 1 1 0 1>}.RGB) {g=true, r=true})
  n2 = Panner(
   Coordinate=@n1
   Speed=AppendVector(A=ScalarParameter() {defaultValue=0.02, parameterName=FoamSpeedX}, B=ScalarParameter() {defaultValue=0.02, parameterName=FoamSpeedY})
  ) {speedX=0.02, speedY=0.02}
  n3 = ParticleRandom()
  n4 = Constant2Vector() {g=32, r=1}
  n5 = TextureSample!(UVs=Add(A=@n2, B=Multiply(A=@n3, B=@n4)), Tex, Level, Apply View MipBias) {mipValueMode=TMVM_MipLevel, texture=/Game/Effects/Fort_Effects/Textures/WaterFoam0001_19_M}
  n6 = FeatureLevelSwitch(
   Default=@n5.RGBA
   ES2=TextureSample!(UVs=@n1, Tex, Apply View MipBias) {texture=&WaterFoam0001_19_M}.RGBA
   SM4_REMOVED=@n5.RGBA
  )
  n7 = StaticComponentMaskParameter(None=@n6) {defaultB=true, defaultG=true, defaultR=true, parameterName=ColorChannel}
  n8 = ParticleColor()
  MaterialFunctionCall(Emissive=Multiply(
    A=MaterialFunctionCall() {materialFunction=&Color_Time_of_Day}.Result
    B=Multiply(A=@n7, B=@n8.RGB)
   ) {constA=1}) {materialFunction=&MF_FX_EmissiveAdjustments}
  n9 = LinearInterpolate(A=ComponentMask(None=@n6) {a=true}) {constB=0.33}
  n10 = TextureCoordinate()
  n11 = Add(A=@n10, B=Multiply(A=@n3, B=Constant2Vector() {g=1, r=32}))
  n12 = SphereMask(
   A=@n10
   B=Constant() {r=0.5}
   Radius=LinearInterpolate(Alpha=TextureSample(UVs=@n11) {samplerType=SAMPLERTYPE_LinearColor, texture=&T_NoisePack_LinearColor}.R) {constA=0.15, constB=0.5}
  ) {attenuationRadius=0.5, hardnessPercent=0.33}
  n13 = DynamicParameter() {defaultValue=<rgba 0 1 1 1>, paramNames [Erosion,Random,Param3,Param4]}
  Multiply(A=Subtract(A=Multiply(A=@n9, B=@n12), B=Saturate(None=@n13.Erosion)), B=@n8.A)
  Multiply(A=@n2, B=ScalarParameter() {defaultValue=1, parameterName=Foam_Texture_tile})
 ]
}
