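# Niagara screen-space-style lens effect for the Buttercake blowhole launch, placed by the User.PlaneOffset vector.
# ScreenSmoke bursts a short-lived sprite animated through curve-driven dynamic material parameters, with its Color module disabled.
# GPU-simulated VerticalLines spawns thin streaks at a decaying rate in an offset hemisphere, shooting them up world Z with drag and alpha fade.
# A LaunchSpikes burst emitter with tinted, stretching sprites is present but disabled.
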NiagaraSystem /ButterCake/VFX/BlowholeLaunch/NS_Buttercake_BlowholeLaunch_LensFX
dependencies {
 usedDataInterfaces [&NiagaraDataInterfaceColorCurve, &NiagaraDataInterfaceCurve]
 usedDynamicInputs []
 usedModules [&SystemState]
}
emitters {
 LaunchSpikes {
  bEnabled=false
  emitterUpdateScript [
   &EmitterState()
   &SpawnBurst_Instantaneous(Use Spawn Probability=false, UseLoopCountLimit=false)
  ]
  particleSpawnScript [&InitializeParticle(
    Color Mode=::NewEnumerator0
    Lifetime=0.1
    Position Offset=&Add_Vector(A=$User.PlaneOffset, B=<0 0 -19.4712>)
    Sprite Size=<50 13.9502>
    Sprite Size Mode=::NewEnumerator4
    UsePositionOffset=true
    Write Material Random=true
   )]
  particleUpdateScript [
   &ParticleState()
   &Color(Color=&ColorFromCurve(ColorCurve=DataInterface() {propertyValues {
       AlphaCurve {keys [
         {arriveTangent=-1.8839, interpMode=RCIM_Cubic, leaveTangent=-1.8839, tangentMode=RCTM_User, time=0, value=1}
         {arriveTangent=-1, interpMode=RCIM_Cubic, leaveTangent=-1, time=1, value=0}
        ]}
       BlueCurve {keys [{arriveTangent=1, interpMode=RCIM_Cubic, leaveTangent=1, time=0, value=1}]}
       GreenCurve {keys [{arriveTangent=0.0801, interpMode=RCIM_Cubic, leaveTangent=0.0801, time=0, value=0.8694}]}
       RedCurve {keys [<at 0 0.4316 RCIM_Cubic>]}
      }}))
   &DynamicMaterialParameters(
    Index 0 Param 1=1
    Index 0 Param 2=&FloatFromCurve(FloatCurve=DataInterface() {propertyValues {Curve {keys [
         {arriveTangent=1.5, interpMode=RCIM_Cubic, leaveTangent=1.5, time=0, value=0.5}
         {arriveTangent=1.5, interpMode=RCIM_Cubic, leaveTangent=1.5, time=1, value=2}
        ]}}})
    Index 0 Param 3=&FloatFromCurve(FloatCurve=DataInterface() {propertyValues {Curve {keys [<at 0 5>,<at 1 1>]}}})
    Module.Param0WriteEnabled=true
    Module.Param1WriteEnabled=true
    Module.Param2WriteEnabled=true
    Module.Param3WriteEnabled=true
   )
   &ScaleSpriteSize(
    Scale Factor=&MakeVector2D(
     X=1
     Y=&FloatFromCurve(FloatCurve=DataInterface() {propertyValues {Curve {keys [
          {arriveTangent=7.5, interpMode=RCIM_Cubic, leaveTangent=7.5, time=0, value=1}
          {arriveTangent=-1, interpMode=RCIM_Cubic, leaveTangent=-1, time=0.2, value=2.5}
          {arriveTangent=-3.125, interpMode=RCIM_Cubic, leaveTangent=-3.125, time=1, value=0}
         ]}}})
    )
    Scale Sprite Size Mode=::NewEnumerator2
   )
  ]
  rendererClasses [&NiagaraSpriteRendererProperties]
  renderers [{rendererClass=&NiagaraSpriteRendererProperties}]
 }
 ScreenSmoke {
  emitterUpdateScript [
   &EmitterState()
   &SpawnBurst_Instantaneous(Use Spawn Probability=false, UseLoopCountLimit=false)
  ]
  particleSpawnScript [&InitializeParticle(
    Color Mode=::NewEnumerator0
    Lifetime=0.78
    Position Offset=$User.PlaneOffset
    Sprite Size=<50 25>
    Sprite Size Mode=::NewEnumerator4
    UsePositionOffset=true
    Write Material Random=true
   )]
  particleUpdateScript [
   &ParticleState()
   &Color(Color=&ColorFromCurve(ColorCurve=DataInterface() {propertyValues {
       AlphaCurve {keys [
         {arriveTangent=-1.8839, interpMode=RCIM_Cubic, leaveTangent=-1.8839, tangentMode=RCTM_User, time=0, value=1}
         {arriveTangent=-1, interpMode=RCIM_Cubic, leaveTangent=-1, time=1, value=0}
        ]}
       BlueCurve {keys [<at 0 1 RCIM_Cubic>,<at 1 1 RCIM_Cubic>]}
       GreenCurve {keys [<at 0 1 RCIM_Cubic>,<at 1 1 RCIM_Cubic>]}
       RedCurve {keys [<at 0 1 RCIM_Cubic>,<at 1 1 RCIM_Cubic>]}
      }})) {enabled=false}
   &DynamicMaterialParameters(
    Index 0 Param 1=0.7
    Index 0 Param 2=&FloatFromCurve(FloatCurve=DataInterface() {propertyValues {Curve {keys [
         {arriveTangent=1.5, interpMode=RCIM_Cubic, leaveTangent=1.5, time=0, value=0.5}
         {arriveTangent=1.5, interpMode=RCIM_Cubic, leaveTangent=1.5, time=1, value=2}
        ]}}})
    Index 0 Param 3=&FloatFromCurve(FloatCurve=DataInterface() {propertyValues {Curve {keys [<at 0 5>,<at 1 1>]}}})
    Module.Param0WriteEnabled=true
    Module.Param1WriteEnabled=true
    Module.Param2WriteEnabled=true
    Module.Param3WriteEnabled=true
   )
  ]
  rendererClasses [&NiagaraSpriteRendererProperties]
  renderers [{rendererClass=&NiagaraSpriteRendererProperties}]
 }
 VerticalLines {
  simTarget=GPUComputeSim
  emitterUpdateScript [
   &EmitterState(Life Cycle Mode=::NewEnumerator1, Loop Behavior=::NewEnumerator1, Loop Duration=0.7, UseLoopDelay=false)
   &SpawnRate(
    SpawnRate=&FloatFromCurve(FloatCurve=DataInterface() {propertyValues {Curve {keys [<at 0 1>,<at 1 0>]}}}, CurveIndex=$Emitter.NormalizedLoopAge, Scale Curve=125)
    Use Spawn Probability=false
   )
  ]
  particleSpawnScript [
   &InitializeParticle(
    Color=<rgba 1 1 1 0.3>
    Lifetime Max=1.75
    Lifetime Mode=::NewEnumerator1
    Mass Mode=::NewEnumerator3
    Sprite Rotation Mode=::NewEnumerator2
    Sprite Size Max=<0.16 11>
    Sprite Size Min=<0.125 4>
    Sprite Size Mode=::NewEnumerator2
    UsePositionOffset=false
    Write Material Random=true
   )
   &AddVelocity(Coordinate Space=::World, Velocity=&RandomRangeVector(Minimum=<0 0 500>, Maximum=<0 0 2500>))
   &SphereLocation(
    Hemisphere Z=true
    MaskSpawnGroup=false
    Offset=&Add_Vector(A=$User.PlaneOffset, B=<0 0 -90>)
    Override Seed=false
    Sphere Radius=16
    Surface Only=false
    UseNonUniformScale=false
    UseOffset=true
    UseSphereOrientation=false
   )
  ]
  particleUpdateScript [
   &ParticleState()
   &Drag(Drag=0.25, Use Linear Drag=true, Use Rotational Drag=false)
   &ScaleColor(
    Scale Alpha=&FloatFromCurve(FloatCurve=DataInterface() {propertyValues {Curve {keys [<at -0.0046 1.0113 RCIM_Cubic>,<at 1.0011 0.002 RCIM_Cubic>]}}}, Scale Curve=0.28)
    ScaleA=true
    ScaleRGB=true
   )
   &SolveForcesAndVelocity(Clamp Velocity=false, Limit Acceleration=false)
  ]
  rendererClasses [&NiagaraSpriteRendererProperties]
  renderers [{rendererClass=&NiagaraSpriteRendererProperties}]
 }
}
summary {
 systemName=NS_Buttercake_BlowholeLaunch_LensFX
 emitters [
  {emitterName=ScreenSmoke, rendererClasses [&NiagaraSpriteRendererProperties]}
  {emitterName=VerticalLines, simTarget=GPUComputeSim, rendererClasses [&NiagaraSpriteRendererProperties]}
  {bEnabled=false, emitterName=LaunchSpikes, rendererClasses [&NiagaraSpriteRendererProperties]}
 ]
 userVariables [User.PlaneOffset {type=Vector3f, description="", defaultValue {struct=Vector3f, value=<20 0 0>}}]
}
